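# Physics asset for the Candor pickaxe cosmetic mesh, defining collision bodies and joints for its dangling straps, D-rings and wires.
# A long capsule on the two-handed handle anchors two strap-to-D-ring chains and two five-link wire chains.
# Small spheres and capsules cover the dyn_ bones, each constrained to its parent link with limited swing and tight twist.
# Straps swing mostly along one axis, while wires swing evenly on both.
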PhysicsAsset /BRCosmetics/Weapons/FORT_Melee/Pickaxe_Candor/Meshes/Candor_Axe_PhysicsAsset
bodies {
 Melee_TwoHanded_Handle [{shapeType=Capsule, length=87.9579, radius=2.3058, shapeName=Melee_TwoHanded_Handle_capsule, center=<0.6855 -1.0278 19.5403>}]
 dyn_strap_lower [{shapeType=Sphere, radius=1.255, shapeName=dyn_strap_lower_sphere, center=<0.0995 0 0.0156>}]
 dyn_strap_lower_D_ring [{shapeType=Capsule, length=2.755, radius=1.255, shapeName=dyn_strap_lower_D_ring_capsule, center=<1.3573 0 -0.0047>}]
 dyn_strap_upper [{shapeType=Sphere, radius=1.255, shapeName=dyn_strap_upper_sphere, center=<0.0661 0 0>}]
 dyn_strap_upper_D_ring [{shapeType=Capsule, length=3.005, radius=1.255, shapeName=dyn_strap_upper_D_ring_capsule, center=<0.9972 0 0.0188>}]
 dyn_wire_left_1 [{shapeType=Sphere, radius=0.755, shapeName=dyn_wire_left_1_sphere, center=<0.1467 -0.1573 0.0144>}]
 dyn_wire_left_2 [{shapeType=Sphere, radius=0.755, shapeName=dyn_wire_left_2_sphere, center=<0.3579 0.027 0.0067>}]
 dyn_wire_left_3 [{shapeType=Capsule, length=1.0416, radius=0.993, shapeName=dyn_wire_left_3_capsule, center=<1.0217 0.0095 0.0086>, rotation=<rot 90 180 180>}]
 dyn_wire_left_4 [{shapeType=Sphere, radius=0.755, shapeName=dyn_wire_left_4_sphere, center=<1.111 0.0102 0.0066>}]
 dyn_wire_left_5 [{shapeType=Sphere, radius=0.755, shapeName=dyn_wire_left_5_sphere, center=<1.4437 0.0116 -0.023>}]
 dyn_wire_right_1 [{shapeType=Sphere, radius=0.505, shapeName=dyn_wire_right_1_sphere, center=<0.2683 -0.1317 -0.0054>}]
 dyn_wire_right_2 [{shapeType=Sphere, radius=0.505, shapeName=dyn_wire_right_2_sphere, center=<0.0248 0.0735 -0.0056>}]
 dyn_wire_right_3 [{
   shapeType=Capsule
   length=1.0284
   radius=0.7431
   shapeName=dyn_wire_right_3_capsule
   center=<1.0086 0.0095 -0.0075>
   rotation=<rot 90 180 180>
  }]
 dyn_wire_right_4 [{shapeType=Sphere, radius=0.505, shapeName=dyn_wire_right_4_sphere, center=<0.5214 0.0102 -0.0054>}]
 dyn_wire_right_5 [{shapeType=Sphere, radius=0.505, shapeName=dyn_wire_right_5_sphere, center=<1.0854 0.0279 -0.0044>}]
}
constraints [
 {bone1Name=dyn_strap_lower, bone2Name=Melee_TwoHanded_Handle, swing1LimitDegrees=50, swing2LimitDegrees=12, twistLimitDegrees=5}
 {bone1Name=dyn_strap_lower_D_ring, bone2Name=dyn_strap_lower, swing1LimitDegrees=50, swing2LimitDegrees=10, twistLimitDegrees=10}
 {bone1Name=dyn_strap_upper, bone2Name=Melee_TwoHanded_Handle, swing1LimitDegrees=50, swing2LimitDegrees=12, twistLimitDegrees=5}
 {bone1Name=dyn_strap_upper_D_ring, bone2Name=dyn_strap_upper, swing1LimitDegrees=50, swing2LimitDegrees=10, twistLimitDegrees=5}
 {bone1Name=dyn_wire_left_1, bone2Name=Melee_TwoHanded_Handle, swing1LimitDegrees=20, swing2LimitDegrees=20, twistLimitDegrees=5}
 {bone1Name=dyn_wire_left_2, bone2Name=dyn_wire_left_1, swing1LimitDegrees=35, swing2LimitDegrees=35, twistLimitDegrees=5}
 {bone1Name=dyn_wire_left_4, bone2Name=dyn_wire_left_3, swing1LimitDegrees=30, swing2LimitDegrees=30, twistLimitDegrees=5}
 {bone1Name=dyn_wire_left_5, bone2Name=dyn_wire_left_4, swing1LimitDegrees=30, swing2LimitDegrees=30, twistLimitDegrees=5}
 {bone1Name=dyn_wire_left_3, bone2Name=dyn_wire_left_2, swing1LimitDegrees=30, swing2LimitDegrees=30, twistLimitDegrees=5}
 {bone1Name=dyn_wire_right_1, bone2Name=Melee_TwoHanded_Handle, swing1LimitDegrees=20, swing2LimitDegrees=20, twistLimitDegrees=5}
 {bone1Name=dyn_wire_right_2, bone2Name=dyn_wire_right_1, swing1LimitDegrees=35, swing2LimitDegrees=35, twistLimitDegrees=10}
 {bone1Name=dyn_wire_right_4, bone2Name=dyn_wire_right_3, swing1LimitDegrees=30, swing2LimitDegrees=30, twistLimitDegrees=5}
 {bone1Name=dyn_wire_right_5, bone2Name=dyn_wire_right_4, swing1LimitDegrees=30, swing2LimitDegrees=30, twistLimitDegrees=5}
 {bone1Name=dyn_wire_right_3, bone2Name=dyn_wire_right_2, swing1LimitDegrees=30, swing2LimitDegrees=30, twistLimitDegrees=5}
]
